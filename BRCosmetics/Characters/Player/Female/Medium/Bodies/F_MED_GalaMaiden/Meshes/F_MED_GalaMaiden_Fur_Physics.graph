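# Physics asset for the fur on the Gala Maiden female medium player body cosmetic, covering three left-side fur chains.
# Each dyn_fur capsule body is constrained to its small base_fur capsule, with twist locked and narrow per-chain swing limit angles.
# An unconstrained upper-left-arm capsule is also included, apparently as a collision volume for the fur.
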PhysicsAsset /BRCosmetics/Characters/Player/Female/Medium/Bodies/F_MED_GalaMaiden/Meshes/F_MED_GalaMaiden_Fur_Physics
bodies {
 base_fur_01_l [{shapeType=Capsule, length=0.505, radius=0.505, shapeName=base_fur_01_l_capsule, rotation=<rot 0 -90 0>}]
 base_fur_02_l [{shapeType=Capsule, length=0.505, radius=0.505, shapeName=base_fur_02_l_capsule, rotation=<rot 0 -90 0>}]
 base_fur_03_l [{shapeType=Capsule, length=0.505, radius=0.505, shapeName=base_fur_03_l_capsule, rotation=<rot 0 -90 0>}]
 dyn_fur_01_l [{shapeType=Capsule, length=8.5, radius=2.5, shapeName=dyn_fur_01_l_capsule, center=<10.7479 5.9922 1.4291>, rotation=<rot 65 -0.0002 -0.0002>}]
 dyn_fur_02_l [{shapeType=Capsule, length=6, radius=2, shapeName=dyn_fur_02_l_capsule, center=<11.4252 -7.4189 -0.8485>, rotation=<rot 60 -0.0001 -0.0002>}]
 dyn_fur_03_l [{shapeType=Capsule, length=11.5, radius=3.5, shapeName=dyn_fur_03_l_capsule, center=<10 -0.716 4.8047>, rotation=<rot 70 -0.0005 -0.0006>}]
 upperarm_l [{shapeType=Capsule, length=20, radius=5, center=<14.0302 0 0>, rotation=<rot 90 0 0>}]
}
constraints [
 {bone1Name=dyn_fur_01_l, bone2Name=base_fur_01_l, swing1LimitDegrees=10, swing2LimitDegrees=10, twistLimitDegrees=45, twistMotion=Locked}
 {bone1Name=dyn_fur_02_l, bone2Name=base_fur_02_l, swing1LimitDegrees=5, swing2LimitDegrees=10, twistLimitDegrees=45, twistMotion=Locked}
 {bone1Name=dyn_fur_03_l, bone2Name=base_fur_03_l, swing1LimitDegrees=5, swing2LimitDegrees=15, twistLimitDegrees=45, twistMotion=Locked}
]
