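# Physics asset for the skirt of the medium female Gala Maiden player body, built entirely from capsule bodies.
# Thigh and calf capsules and one oversized capsule offset below the root accompany the skirt bodies.
# Asymmetric skirt chains cover left and right back and left side; the front right has only a tiny base body.
# Most root and anchor joints swing and twist freely, while lower skirt joints have tight swing limits, some with locked twist.
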PhysicsAsset /BRCosmetics/Characters/Player/Female/Medium/Bodies/F_MED_GalaMaiden/Meshes/F_MED_GalaMaiden_Skirt_Physics
bodies {
 base_back_skirt_middle_01_r [{shapeType=Capsule, length=0.505, radius=0.505, shapeName=base_back_skirt_middle_01_r_capsule, rotation=<rot 0 -90 0>}]
 base_front_skirt_r [{shapeType=Capsule, length=0.505, radius=0.505, shapeName=base_front_skirt_r_capsule, rotation=<rot 0 -90 0>}]
 calf_l [{shapeType=Capsule, length=33.3174, radius=5.2876, center=<-27.3271 -2.6326 -0.9066>, rotation=<rot 0 84.1189 -88.1473>}]
 calf_r [{shapeType=Capsule, length=49.9867, radius=7.1182, center=<23.8007 3.7384 0.7895>, rotation=<rot -0.3217 94.1134 -88.1754>}]
 dyn_back_skirt_middle_01_r [{
   shapeType=Capsule
   length=14.0268
   radius=9.6077
   shapeName=dyn_back_skirt_middle_01_r_capsule
   center=<14.5939 4.9036 -0.4441>
   rotation=<rot 0 -169.4948 -82.0246>
  }]
 dyn_back_skirt_middle_02_l [{
   shapeType=Capsule
   length=13.8735
   radius=4.3417
   shapeName=dyn_back_skirt_middle_02_l_capsule
   center=<9.5724 3.6917 0>
   rotation=<rot 90 180 -90>
  }]
 dyn_back_skirt_middle_02_r [{
   shapeType=Capsule
   length=25.1868
   radius=8.0933
   shapeName=dyn_back_skirt_middle_02_r_capsule
   center=<9.1738 0 11.9253>
   rotation=<rot 0 0 90>
  }]
 dyn_back_skirt_middle_03_l [{
   shapeType=Capsule
   length=14.884
   radius=3.2941
   shapeName=dyn_back_skirt_middle_03_l_capsule
   center=<19.0009 7.0803 1.3792>
   rotation=<rot 90 0 90>
  }]
 dyn_back_skirt_middle_03_r [{
   shapeType=Capsule
   length=19.7067
   radius=10.2619
   shapeName=dyn_back_skirt_middle_03_r_capsule
   center=<8.4554 -6.8798 3.7132>
   rotation=<rot 0 0 90>
  }]
 dyn_side_skirt_03_l [{shapeType=Capsule, length=8.5, radius=2, center=<3 0 0>, rotation=<rot 90 0 0>}]
 dyn_side_skirt_04_l [{
   shapeType=Capsule
   length=15.5
   radius=1.5
   shapeName=dyn_side_skirt_04_l_capsule
   center=<12.5979 -5.6771 0.4252>
   rotation=<rot 90 -90 -90.0007>
  }]
 offset_back_skirt_middle_01_l [{shapeType=Capsule, length=0.505, radius=0.505, shapeName=offset_back_skirt_middle_01_l_capsule, rotation=<rot 0 -90 0>}]
 offset_back_skirt_middle_01_r [{shapeType=Capsule, length=0.505, radius=0.505, shapeName=offset_back_skirt_middle_01_r_capsule, rotation=<rot 0 -90 0>}]
 root [{shapeType=Capsule, length=196.5, radius=153, shapeName=root_capsule, center=<0 0 -158.963>, rotation=<rot 90 180 -90>}]
 thigh_l [{shapeType=Capsule, length=23.8055, radius=6.5828, shapeName=thigh_l_capsule, center=<-26.0892 0.2775 0.4885>, rotation=<rot 90 0 0>}]
 thigh_r [{shapeType=Capsule, length=23.6338, radius=7.43, center=<21.0266 -1.7016 -0.7397>, rotation=<rot 0 94.7425 90.1872>}]
}
constraints [
 {bone1Name=calf_l, bone2Name=thigh_l, swing1LimitDegrees=45, swing2LimitDegrees=45, twistLimitDegrees=45}
 {
  bone1Name=dyn_side_skirt_03_l
  bone2Name=root
  swing1LimitDegrees=45
  swing1Motion=Free
  swing2LimitDegrees=45
  swing2Motion=Free
  twistLimitDegrees=45
  twistMotion=Free
 }
 {
  bone1Name=calf_r
  bone2Name=thigh_r
  swing1LimitDegrees=45
  swing1Motion=Free
  swing2LimitDegrees=45
  swing2Motion=Free
  twistLimitDegrees=45
  twistMotion=Free
 }
 {
  bone1Name=dyn_side_skirt_04_l
  bone2Name=dyn_side_skirt_03_l
  swing1LimitDegrees=10
  swing2LimitDegrees=20
  twistLimitDegrees=45
  twistMotion=Locked
 }
 {
  bone1Name=thigh_r
  bone2Name=root
  swing1LimitDegrees=45
  swing1Motion=Free
  swing2LimitDegrees=45
  swing2Motion=Free
  twistLimitDegrees=45
  twistMotion=Free
 }
 {
  bone1Name=dyn_back_skirt_middle_02_l
  bone2Name=offset_back_skirt_middle_01_l
  swing1LimitDegrees=10
  swing2LimitDegrees=10
  twistLimitDegrees=0
 }
 {
  bone1Name=dyn_back_skirt_middle_03_l
  bone2Name=dyn_back_skirt_middle_02_l
  swing1LimitDegrees=10
  swing2LimitDegrees=10
  twistLimitDegrees=0
 }
 {
  bone1Name=dyn_back_skirt_middle_02_r
  bone2Name=dyn_back_skirt_middle_01_r
  swing1LimitDegrees=30
  swing2LimitDegrees=20
  twistLimitDegrees=0
  twistMotion=Locked
 }
 {
  bone1Name=dyn_back_skirt_middle_03_r
  bone2Name=dyn_back_skirt_middle_02_r
  swing1LimitDegrees=10
  swing2LimitDegrees=20
  twistLimitDegrees=0
  twistMotion=Locked
 }
 {
  bone1Name=offset_back_skirt_middle_01_l
  bone2Name=root
  swing1LimitDegrees=45
  swing1Motion=Free
  swing2LimitDegrees=45
  swing2Motion=Free
  twistLimitDegrees=45
  twistMotion=Free
 }
 {
  bone1Name=offset_back_skirt_middle_01_r
  bone2Name=base_back_skirt_middle_01_r
  swing1LimitDegrees=45
  swing1Motion=Free
  swing2LimitDegrees=45
  swing2Motion=Free
  twistLimitDegrees=45
  twistMotion=Free
 }
 {
  bone1Name=thigh_l
  bone2Name=root
  swing1LimitDegrees=45
  swing1Motion=Free
  swing2LimitDegrees=45
  swing2Motion=Free
  twistLimitDegrees=45
  twistMotion=Free
 }
 {
  bone1Name=base_front_skirt_r
  bone2Name=root
  swing1LimitDegrees=45
  swing1Motion=Free
  swing2LimitDegrees=45
  swing2Motion=Free
  twistLimitDegrees=45
  twistMotion=Free
 }
 {
  bone1Name=dyn_back_skirt_middle_01_r
  bone2Name=offset_back_skirt_middle_01_r
  swing1LimitDegrees=45
  swing1Motion=Free
  swing2LimitDegrees=45
  swing2Motion=Free
  twistLimitDegrees=45
  twistMotion=Free
 }
 {
  bone1Name=base_back_skirt_middle_01_r
  bone2Name=root
  swing1LimitDegrees=45
  swing1Motion=Free
  swing2LimitDegrees=45
  swing2Motion=Free
  twistLimitDegrees=45
  twistMotion=Free
 }
]
