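# Animation Blueprint for the Gala Maiden medium female cosmetic body, adding secondary motion to skirt and fur.
# It copies the parent mesh pose into a Control Rig, fed crouch-idle and emote flags.
# Chained cached poses apply additive bone corrections and a belt-gun constraint, then pelvis- and spine-relative RigidBody simulations and final skirt tweaks.
# The event graph updates gravity override through a deprecated helper, plus movement and specific-emote flags.
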AnimBlueprint /BRCosmetics/Characters/Player/Female/Medium/Bodies/F_MED_GalaMaiden/Meshes/F_MED_GalaMaiden_AnimBP
target_skeleton=&/F_MED_GalaMaiden_Skeleton
graphs [
 AnimGraph {nodes [
   n1 = RigidBody!(ComponentPose=Local To Component(LocalPose=Use cached pose 'FixPose'()), Alpha, OverrideWorldGravity:Vector="0,0,0", ExternalForce) -> (Pose) {
    bClampLinearTranslationLimitToRefPose=true
    bOverrideWorldGravity=true
    overridePhysicsAsset=&/F_MED_GalaMaiden_Skirt_Physics
    simulationSpace=BaseBoneSpace
    alphaBoolBlend {blendOption=Linear}
    baseBoneRef {boneName=pelvis}
    componentLinearAccScale=<0.3 0.3 0.3>
    componentLinearVelScale=<0.3 0.3 0.3>
    simSpaceSettings {velocityScaleZ=0, worldAlpha=0.05}
   }
   n2 = RigidBody(ComponentPose=@n1, OverrideWorldGravity="0,0,0") {
    bClampLinearTranslationLimitToRefPose=true
    bOverrideWorldGravity=true
    overridePhysicsAsset=&/F_MED_GalaMaiden_Fur_Physics
    simulationSpace=BaseBoneSpace
    alphaBoolBlend {blendOption=Linear}
    baseBoneRef {boneName=spine_05}
    componentLinearAccScale=<0.3 0.1 0.1>
    componentLinearVelScale=<0.3 0.1 0.1>
    simSpaceSettings {worldAlpha=0.07}
   }
   n3 = "Transform (Modify) Bone - Bone: offset_back_skirt_middle_01_r"!(ComponentPose=@n2, bAlphaBoolEnabled:bool=True, Translation="0,0,0", Rotation="-4.785,-9.985,1.949", Scale="1,1,1") -> (Pose) {
    alphaInputType=Bool
    rotationMode=BMM_Additive
    rotationSpace=BCS_BoneSpace
    alphaBoolBlend {blendInTime=0.2, blendOption=Linear, blendOutTime=0.2}
    boneToModify {boneName=offset_back_skirt_middle_01_r}
    rotation=<rot -4.785 -9.985 1.949>
   }
   n4 = "Transform (Modify) Bone - Bone: offset_back_skirt_middle_01_l"!(ComponentPose=@n3, bAlphaBoolEnabled:bool=True, Translation="0,0,0", Rotation="-9.771,-15.234,3.130", Scale="1,1,1") -> (Pose) {
    alphaInputType=Bool
    rotationMode=BMM_Additive
    rotationSpace=BCS_BoneSpace
    alphaBoolBlend {blendInTime=0.2, blendOption=Linear, blendOutTime=0.2}
    boneToModify {boneName=offset_back_skirt_middle_01_l}
    rotation=<rot -9.771 -15.234 3.13>
   }
   n5 = "Transform (Modify) Bone - Bone: dyn_back_skirt_middle_02_r"!(ComponentPose=@n4, bAlphaBoolEnabled:bool=True, Translation="0,0,0", Rotation="20,0,0", Scale="1,1,1") -> (Pose) {
    alphaInputType=Bool
    rotationMode=BMM_Additive
    rotationSpace=BCS_BoneSpace
    translationMode=BMM_Additive
    translationSpace=BCS_BoneSpace
    alphaBoolBlend {blendInTime=0.2, blendOption=Linear, blendOutTime=0.2}
    boneToModify {boneName=dyn_back_skirt_middle_02_r}
    rotation {pitch=20}
   }
   n6 = "Transform (Modify) Bone - Bone: dyn_back_skirt_middle_02_r"(ComponentPose=@n5, bAlphaBoolEnabled=True, Translation="0,0,0", Rotation="0,0,0", Scale="1,1,1") {
    alphaInputType=Bool
    rotationMode=BMM_Additive
    rotationSpace=BCS_BoneSpace
    translationMode=BMM_Additive
    translationSpace=BCS_BoneSpace
    alphaBoolBlend {blendInTime=0.2, blendOption=Linear, blendOutTime=0.2}
    boneToModify {boneName=dyn_back_skirt_middle_02_r}
   }
   n7 = "Transform (Modify) Bone - Bone: dyn_back_skirt_middle_02_r"(ComponentPose=@n6, bAlphaBoolEnabled=True, Translation="0,0,0", Rotation="20,-20,0.004", Scale="1,1,1") {
    alphaInputType=Bool
    rotationMode=BMM_Additive
    rotationSpace=BCS_BoneSpace
    translationMode=BMM_Additive
    translationSpace=BCS_BoneSpace
    alphaBoolBlend {blendInTime=0.2, blendOption=Linear, blendOutTime=0.2}
    boneToModify {boneName=dyn_back_skirt_middle_02_r}
    rotation=<rot 20 -20 0.004>
   }
   Output Pose(Result=Component To Local(ComponentPose=@n7)) {layerGroup=None}
   n8 = Control Rig!(
    Source=Copy Pose From Mesh() {bUseAttachedParent=true}
    Alpha
    Crouch:bool=Get CrouchIDLE() -> (CrouchIDLE:bool)
    IsEmote:bool=Get bIsPlayingEmote()
   ) -> (Pose) {
    alphaBoolBlend {blendOption=Linear}
    controlRigAssetReference {blueprintRigClass=&/F_MED_GalaMaiden_CtrlRig.F_MED_GalaMaiden_CtrlRig_C}
   }
   Save cached pose 'ControlRigPose'(Pose=@n8) {cacheName=ControlRigPose}
   n9 = "Transform (Modify) Bone - Bone: offset_back_skirt_middle_01_r"(
    ComponentPose=Local To Component(LocalPose=Use cached pose 'MovingPose'())
    bAlphaBoolEnabled=True
    Translation="-2.076,-0.4,0"
    Rotation="-5.000,-10.000,-0.005"
    Scale="1,1,1"
   ) {
    alphaInputType=Bool
    rotationMode=BMM_Additive
    rotationSpace=BCS_BoneSpace
    translationMode=BMM_Additive
    translationSpace=BCS_BoneSpace
    alphaBoolBlend {blendInTime=0.2, blendOption=Linear, blendOutTime=0.2}
    boneToModify {boneName=offset_back_skirt_middle_01_r}
    rotation=<rot -5 -10 -0.005>
    translation=<-2.076 -0.4>
   }
   n10 = "Transform (Modify) Bone - Bone: base_front_skirt_l"!(ComponentPose=@n9, bAlphaBoolEnabled:bool=True, Translation="-4.316,0.699,1.862", Rotation="-23.719,4.853,-12.006", Scale="1,1,1") -> (Pose) {
    alphaInputType=Bool
    rotationMode=BMM_Additive
    rotationSpace=BCS_BoneSpace
    translationMode=BMM_Additive
    translationSpace=BCS_BoneSpace
    alphaBoolBlend {blendInTime=0.2, blendOption=Linear, blendOutTime=0.1}
    boneToModify {boneName=base_front_skirt_l}
    rotation=<rot -23.719 4.853 -12.006>
    translation=<-4.316 0.699 1.862>
   }
   n11 = "Transform (Modify) Bone - Bone: base_side_skirt_r"!(ComponentPose=@n10, bAlphaBoolEnabled:bool=True, Translation="0,0,0", Rotation="-10,5.063,-0.045", Scale="1,1,1") -> (Pose) {
    alphaInputType=Bool
    rotationMode=BMM_Additive
    rotationSpace=BCS_BoneSpace
    translationMode=BMM_Additive
    translationSpace=BCS_BoneSpace
    alphaBoolBlend {blendInTime=0.2, blendOption=Linear, blendOutTime=0.2}
    boneToModify {boneName=base_side_skirt_r}
    rotation=<rot -10 5.063 -0.045>
   }
   n12 = "Transform (Modify) Bone - Bone: base_side_skirt_r"(ComponentPose=@n11, bAlphaBoolEnabled=True, Translation="0,0,0", Rotation="7,-15.005,0.012", Scale="1,1,1") {
    alphaInputType=Bool
    rotationMode=BMM_Additive
    rotationSpace=BCS_BoneSpace
    translationMode=BMM_Additive
    translationSpace=BCS_BoneSpace
    alphaBoolBlend {blendInTime=0.2, blendOption=Linear, blendOutTime=0.2}
    boneToModify {boneName=base_side_skirt_r}
    rotation=<rot 7 -15.005 0.012>
   }
   n13 = "Transform (Modify) Bone - Bone: dyn_belt_gun_r"!(ComponentPose=@n12, bAlphaBoolEnabled:bool=True, Translation="0.295,-3.758,-0.863", Rotation="-10,0.172,0.27", Scale="1,1,1") -> (Pose) {
    alphaInputType=Bool
    rotationMode=BMM_Additive
    rotationSpace=BCS_BoneSpace
    translationMode=BMM_Additive
    translationSpace=BCS_BoneSpace
    alphaBoolBlend {blendInTime=0.2, blendOption=Linear, blendOutTime=0.2}
    boneToModify {boneName=dyn_belt_gun_r}
    rotation=<rot -10 0.172 0.27>
    translation=<0.295 -3.758 -0.863>
   }
   n14 = "Constraint - base_belt_gun_r (PPP)"!(ComponentPose=@n13, Alpha=1, ConstraintWeights_0=0.6028, ConstraintWeights_1:float=0.1992, ConstraintWeights_2:float=0.198) -> (Pose) {
    alphaBoolBlend {blendOption=Linear}
    boneToModify {boneName=base_belt_gun_r}
    constraintSetup [{targetBone {boneName=pelvis}}, {targetBone {boneName=deform_groin_r}}, {targetBone {boneName=thigh_twist_01_r}}]
    constraintWeights [1,1,1]
   }
   n15 = "Transform (Modify) Bone - Bone: offset_back_skirt_middle_01_r"(ComponentPose=@n14, bAlphaBoolEnabled=True, Translation="0,0,0", Rotation="0,0,-30", Scale="1,1,1") {
    alphaInputType=Bool
    rotationMode=BMM_Additive
    scaleMode=BMM_Additive
    translationMode=BMM_Additive
    alphaBoolBlend {blendInTime=0.2, blendOption=Linear, blendOutTime=0.2}
    boneToModify {boneName=offset_back_skirt_middle_01_r}
    rotation {roll=-30}
   }
   n16 = "Transform (Modify) Bone - Bone: offset_back_skirt_middle_01_l"(ComponentPose=@n15, bAlphaBoolEnabled=True, Translation="0,0,0", Rotation="0,0,-10", Scale="1,1,1") {
    alphaInputType=Bool
    rotationMode=BMM_Additive
    scaleMode=BMM_Additive
    translationMode=BMM_Additive
    alphaBoolBlend {blendInTime=0.2, blendOption=Linear, blendOutTime=0.2}
    boneToModify {boneName=offset_back_skirt_middle_01_l}
    rotation {roll=-10}
   }
   n17 = "Transform (Modify) Bone - Bone: base_side_skirt_l"!(ComponentPose=@n16, bAlphaBoolEnabled:bool=True, Translation="0,0,0", Rotation="0,0,-30", Scale="1,1,1") -> (Pose) {
    alphaInputType=Bool
    rotationMode=BMM_Additive
    scaleMode=BMM_Additive
    translationMode=BMM_Additive
    alphaBoolBlend {blendInTime=0.2, blendOption=Linear, blendOutTime=0.2}
    boneToModify {boneName=base_side_skirt_l}
    rotation {roll=-30}
   }
   n18 = "Transform (Modify) Bone - Bone: base_side_skirt_r"(ComponentPose=@n17, bAlphaBoolEnabled=True, Translation="0,0,0", Rotation="10,0,-5", Scale="1,1,1") {
    alphaInputType=Bool
    rotationMode=BMM_Additive
    scaleMode=BMM_Additive
    translationMode=BMM_Additive
    alphaBoolBlend {blendInTime=0.2, blendOption=Linear, blendOutTime=0.2}
    boneToModify {boneName=base_side_skirt_r}
    rotation {pitch=10, roll=-5}
   }
   n19 = "Transform (Modify) Bone - Bone: base_front_skirt_01_m"!(ComponentPose=@n18, bAlphaBoolEnabled:bool=True, Translation="2.905,-0.010,1.331", Rotation="3.627,-6.288,-24.628", Scale="1,1,1") -> (Pose) {
    alphaInputType=Bool
    rotationMode=BMM_Additive
    rotationSpace=BCS_BoneSpace
    translationMode=BMM_Additive
    translationSpace=BCS_BoneSpace
    alphaBoolBlend {blendInTime=0.2, blendOption=Linear, blendOutTime=0.2}
    boneToModify {boneName=base_front_skirt_01_m}
    rotation=<rot 3.6271 -6.288 -24.628>
    translation=<2.905 -0.01 1.331>
   }
   Save cached pose 'TransformPose'(Pose=Component To Local(ComponentPose=@n19)) {cacheName=TransformPose}
   n20 = "Transform (Modify) Bone - Bone: base_front_skirt_l"(
    ComponentPose=Local To Component(LocalPose=Use cached pose 'EmotePose'())
    bAlphaBoolEnabled=True
    Translation="0,0,0"
    Rotation="10.001,-5.001,-0.000"
    Scale="1,1,1"
   ) {
    alphaInputType=Bool
    rotationMode=BMM_Additive
    rotationSpace=BCS_BoneSpace
    translationMode=BMM_Additive
    translationSpace=BCS_BoneSpace
    alphaBoolBlend {blendInTime=0.2, blendOption=Linear, blendOutTime=0.2}
    boneToModify {boneName=base_front_skirt_l}
    rotation=<rot 10.0011 -5.001 0>
   }
   n21 = "Transform (Modify) Bone - Bone: base_side_skirt_l"(ComponentPose=@n20, bAlphaBoolEnabled=True, Translation="0,0,0", Rotation="0.001,-5.001,-0.000", Scale="1,1,1") {
    alphaInputType=Bool
    rotationMode=BMM_Additive
    rotationSpace=BCS_BoneSpace
    translationMode=BMM_Additive
    translationSpace=BCS_BoneSpace
    alphaBoolBlend {blendInTime=0.2, blendOption=Linear, blendOutTime=0.2}
    boneToModify {boneName=base_side_skirt_l}
    rotation=<rot 0.001 -5.001 0>
   }
   n22 = "Transform (Modify) Bone - Bone: base_back_skirt_middle_01_l"!(ComponentPose=@n21, bAlphaBoolEnabled:bool=True, Translation="0,0,0", Rotation="60.000,0.002,0.000", Scale="1,1,1") -> (Pose) {
    alphaInputType=Bool
    rotationMode=BMM_Additive
    rotationSpace=BCS_BoneSpace
    translationMode=BMM_Additive
    translationSpace=BCS_BoneSpace
    alphaBoolBlend {blendInTime=0.2, blendOption=Linear, blendOutTime=0.2}
    boneToModify {boneName=base_back_skirt_middle_01_l}
    rotation=<rot 60 0.002 0>
   }
   n23 = "Transform (Modify) Bone - Bone: base_back_skirt_middle_01_r"!(ComponentPose=@n22, bAlphaBoolEnabled:bool=True, Translation="-5.020,0.000,5.797", Rotation="70.000,0.000,0.000", Scale="1,1,1") -> (Pose) {
    alphaInputType=Bool
    rotationMode=BMM_Additive
    rotationSpace=BCS_BoneSpace
    translationMode=BMM_Additive
    translationSpace=BCS_BoneSpace
    alphaBoolBlend {blendInTime=0.2, blendOption=Linear, blendOutTime=0.2}
    boneToModify {boneName=base_back_skirt_middle_01_r}
    rotation=<rot 70 0 0>
    translation {x=-5.02, z=5.797}
   }
   n24 = "Transform (Modify) Bone - Bone: base_fur_l"!(ComponentPose=@n23, bAlphaBoolEnabled:bool=True, Translation="0,0,0", Rotation="-10,-0.134,0", Scale="1,1,1") -> (Pose) {
    alphaInputType=Bool
    rotationMode=BMM_Additive
    rotationSpace=BCS_BoneSpace
    translationMode=BMM_Additive
    translationSpace=BCS_BoneSpace
    alphaBoolBlend {blendInTime=0.2, blendOption=Linear, blendOutTime=0.2}
    boneToModify {boneName=base_fur_l}
    rotation {pitch=-10, yaw=-0.134}
   }
   n25 = "Transform (Modify) Bone - Bone: base_fur_01_l"!(ComponentPose=@n24, bAlphaBoolEnabled:bool=True, Translation="0,0,0", Rotation="-10,0,0", Scale="1,1,1") -> (Pose) {
    alphaInputType=Bool
    rotationMode=BMM_Additive
    rotationSpace=BCS_BoneSpace
    translationMode=BMM_Additive
    translationSpace=BCS_BoneSpace
    alphaBoolBlend {blendInTime=0.2, blendOption=Linear, blendOutTime=0.2}
    boneToModify {boneName=base_fur_01_l}
    rotation {pitch=-10}
   }
   n26 = "Transform (Modify) Bone - Bone: base_fur_02_l"!(ComponentPose=@n25, bAlphaBoolEnabled:bool=True, Translation="0,0,0", Rotation="-10,0,0", Scale="1,1,1") -> (Pose) {
    alphaInputType=Bool
    rotationMode=BMM_Additive
    rotationSpace=BCS_BoneSpace
    translationMode=BMM_Additive
    translationSpace=BCS_BoneSpace
    alphaBoolBlend {blendInTime=0.2, blendOption=Linear, blendOutTime=0.2}
    boneToModify {boneName=base_fur_02_l}
    rotation {pitch=-10}
   }
   n27 = "Transform (Modify) Bone - Bone: base_fur_03_l"!(ComponentPose=@n26, bAlphaBoolEnabled:bool=True, Translation="0,0,0", Rotation="-10,0,0", Scale="1,1,1") -> (Pose) {
    alphaInputType=Bool
    rotationMode=BMM_Additive
    rotationSpace=BCS_BoneSpace
    translationMode=BMM_Additive
    translationSpace=BCS_BoneSpace
    alphaBoolBlend {blendInTime=0.2, blendOption=Linear, blendOutTime=0.2}
    boneToModify {boneName=base_fur_03_l}
    rotation {pitch=-10}
   }
   n28 = "Transform (Modify) Bone - Bone: offset_back_skirt_middle_01_r"(ComponentPose=@n27, bAlphaBoolEnabled=True, Translation="0,0,0", Rotation="-35,-20,0", Scale="1,1,1") {
    alphaInputType=Bool
    rotationMode=BMM_Additive
    rotationSpace=BCS_BoneSpace
    translationMode=BMM_Additive
    translationSpace=BCS_BoneSpace
    alphaBoolBlend {blendInTime=0.2, blendOption=Linear, blendOutTime=0.2}
    boneToModify {boneName=offset_back_skirt_middle_01_r}
    rotation {pitch=-35, yaw=-20}
   }
   n29 = "Transform (Modify) Bone - Bone: base_side_skirt_r"(ComponentPose=@n28, bAlphaBoolEnabled=True, Translation="0,0,0", Rotation="0,-10,0", Scale="1,1,1") {
    alphaInputType=Bool
    rotationMode=BMM_Additive
    rotationSpace=BCS_BoneSpace
    translationMode=BMM_Additive
    translationSpace=BCS_BoneSpace
    alphaBoolBlend {blendInTime=0.2, blendOption=Linear, blendOutTime=0.2}
    boneToModify {boneName=base_side_skirt_r}
    rotation {yaw=-10}
   }
   n30 = "Transform (Modify) Bone - Bone: offset_back_skirt_middle_01_l"(ComponentPose=@n29, bAlphaBoolEnabled=True, Translation="0,0,0", Rotation="-20,0,0", Scale="1,1,1") {
    alphaInputType=Bool
    rotationMode=BMM_Additive
    rotationSpace=BCS_BoneSpace
    translationMode=BMM_Additive
    translationSpace=BCS_BoneSpace
    alphaBoolBlend {blendInTime=0.2, blendOption=Linear, blendOutTime=0.2}
    boneToModify {boneName=offset_back_skirt_middle_01_l}
    rotation {pitch=-20}
   }
   Save cached pose 'FixPose'(Pose=Component To Local(ComponentPose=@n30)) {cacheName=FixPose}
   n31 = "Transform (Modify) Bone - Bone: offset_back_skirt_middle_01_r"(
    ComponentPose=Local To Component(LocalPose=Use cached pose 'ControlRigPose'())
    bAlphaBoolEnabled=True
    Translation="0,0,0"
    Rotation="-5.000,0.215,0.039"
    Scale="1,1,1"
   ) {
    alphaInputType=Bool
    rotationMode=BMM_Additive
    rotationSpace=BCS_BoneSpace
    alphaBoolBlend {blendInTime=0.2, blendOption=Linear, blendOutTime=0.2}
    boneToModify {boneName=offset_back_skirt_middle_01_r}
    rotation=<rot -5.0001 0.215 0.039>
   }
   n32 = "Transform (Modify) Bone - Bone: offset_back_skirt_middle_01_l"(ComponentPose=@n31, bAlphaBoolEnabled=True, Translation="0,0,0", Rotation="-5.000,0.215,0.039", Scale="1,1,1") {
    alphaInputType=Bool
    rotationMode=BMM_Additive
    rotationSpace=BCS_BoneSpace
    alphaBoolBlend {blendInTime=0.2, blendOption=Linear, blendOutTime=0.2}
    boneToModify {boneName=offset_back_skirt_middle_01_l}
    rotation=<rot -5.0001 0.215 0.039>
   }
   n33 = "Transform (Modify) Bone - Bone: base_front_skirt_01_m"(ComponentPose=@n32, bAlphaBoolEnabled=True, Translation="0,0,5", Rotation="0,0,0", Scale="1,1,1") {
    alphaInputType=Bool
    rotationMode=BMM_Additive
    rotationSpace=BCS_BoneSpace
    translationMode=BMM_Additive
    translationSpace=BCS_BoneSpace
    alphaBoolBlend {blendInTime=0.2, blendOption=Linear, blendOutTime=0.2}
    boneToModify {boneName=base_front_skirt_01_m}
    translation {z=5}
   }
   n34 = "Transform (Modify) Bone - Bone: dyn_front_skirt_01_m"!(ComponentPose=@n33, bAlphaBoolEnabled:bool=True, Translation="0,0,0", Rotation="10,0,0", Scale="1,1,1") -> (Pose) {
    alphaInputType=Bool
    rotationMode=BMM_Additive
    rotationSpace=BCS_BoneSpace
    translationMode=BMM_Additive
    translationSpace=BCS_BoneSpace
    alphaBoolBlend {blendInTime=0.2, blendOption=Linear, blendOutTime=0.2}
    boneToModify {boneName=dyn_front_skirt_01_m}
    rotation {pitch=10}
   }
   n35 = "Transform (Modify) Bone - Bone: dyn_front_skirt_02_m"!(ComponentPose=@n34, bAlphaBoolEnabled:bool=True, Translation="-0.537,-10,6.127", Rotation="30,0,0", Scale="1,1,1") -> (Pose) {
    alphaInputType=Bool
    rotationMode=BMM_Additive
    rotationSpace=BCS_BoneSpace
    translationMode=BMM_Additive
    translationSpace=BCS_BoneSpace
    alphaBoolBlend {blendInTime=0.2, blendOption=Linear, blendOutTime=0.2}
    boneToModify {boneName=dyn_front_skirt_02_m}
    rotation {pitch=30}
    translation=<-0.537 -10 6.127>
   }
   n36 = "Transform (Modify) Bone - Bone: offset_back_skirt_middle_01_r"(ComponentPose=@n35, bAlphaBoolEnabled=True, Translation="0,0,0", Rotation="-5,0,0", Scale="1,1,1") {
    alphaInputType=Bool
    rotationMode=BMM_Additive
    rotationSpace=BCS_BoneSpace
    translationMode=BMM_Additive
    translationSpace=BCS_BoneSpace
    alphaBoolBlend {blendInTime=0.2, blendOption=Linear, blendOutTime=0.2}
    boneToModify {boneName=offset_back_skirt_middle_01_r}
    rotation {pitch=-5}
   }
   n37 = "Transform (Modify) Bone - Bone: base_side_skirt_r"(ComponentPose=@n36, bAlphaBoolEnabled=True, Translation="0,0,0", Rotation="10.000,-0.003,0.008", Scale="1,1,1") {
    alphaInputType=Bool
    rotationMode=BMM_Additive
    rotationSpace=BCS_BoneSpace
    translationMode=BMM_Additive
    translationSpace=BCS_BoneSpace
    alphaBoolBlend {blendInTime=0.2, blendOption=Linear, blendOutTime=0.2}
    boneToModify {boneName=base_side_skirt_r}
    rotation=<rot 10 -0.003 0.008>
   }
   n38 = "Transform (Modify) Bone - Bone: offset_back_skirt_middle_01_r"(ComponentPose=@n37, bAlphaBoolEnabled=True, Translation="0,0,0", Rotation="-10.000,0.001,-0.003", Scale="1,1,1") {
    alphaInputType=Bool
    rotationMode=BMM_Additive
    rotationSpace=BCS_BoneSpace
    translationMode=BMM_Additive
    translationSpace=BCS_BoneSpace
    alphaBoolBlend {blendInTime=0.2, blendOption=Linear, blendOutTime=0.2}
    boneToModify {boneName=offset_back_skirt_middle_01_r}
    rotation=<rot -10 0.001 -0.003>
   }
   n39 = "Transform (Modify) Bone - Bone: base_front_skirt_l"(ComponentPose=@n38, bAlphaBoolEnabled=True, Translation="0,0,0", Rotation="5,0,0", Scale="1,1,1") {
    alphaInputType=Bool
    rotationMode=BMM_Additive
    rotationSpace=BCS_BoneSpace
    translationMode=BMM_Additive
    translationSpace=BCS_BoneSpace
    alphaBoolBlend {blendInTime=0.2, blendOption=Linear, blendOutTime=0.2}
    boneToModify {boneName=base_front_skirt_l}
    rotation {pitch=5}
   }
   n40 = "Transform (Modify) Bone - Bone: offset_back_skirt_middle_01_l"(ComponentPose=@n39, bAlphaBoolEnabled=True, Translation="0,0,0", Rotation="-10,0,0", Scale="1,1,1") {
    alphaInputType=Bool
    rotationMode=BMM_Additive
    rotationSpace=BCS_BoneSpace
    translationMode=BMM_Additive
    translationSpace=BCS_BoneSpace
    alphaBoolBlend {blendInTime=0.2, blendOption=Linear, blendOutTime=0.2}
    boneToModify {boneName=offset_back_skirt_middle_01_l}
    rotation {pitch=-10}
   }
   Save cached pose 'MovingPose'(Pose=Component To Local(ComponentPose=@n40)) {cacheName=MovingPose}
   n41 = "Transform (Modify) Bone - Bone: dyn_front_skirt_02_m"(
    ComponentPose=Local To Component(LocalPose=Use cached pose 'TransformPose'())
    bAlphaBoolEnabled=True
    Translation="0,0,-20"
    Rotation="0,0,0"
    Scale="1,1,1"
   ) {
    alphaInputType=Bool
    rotationMode=BMM_Additive
    rotationSpace=BCS_BoneSpace
    translationMode=BMM_Additive
    translationSpace=BCS_BoneSpace
    alphaBoolBlend {blendInTime=0.2, blendOption=Linear, blendOutTime=0.2}
    boneToModify {boneName=dyn_front_skirt_02_m}
    translation {z=-20}
   }
   n42 = "Transform (Modify) Bone - Bone: base_front_skirt_l"(ComponentPose=@n41, bAlphaBoolEnabled=True, Translation="0,0,0", Rotation="5,0,0", Scale="1,1,1") {
    alphaInputType=Bool
    rotationMode=BMM_Additive
    rotationSpace=BCS_BoneSpace
    translationMode=BMM_Additive
    translationSpace=BCS_BoneSpace
    alphaBoolBlend {blendInTime=0.2, blendOption=Linear, blendOutTime=0.2}
    boneToModify {boneName=base_front_skirt_l}
    rotation {pitch=5}
   }
   n43 = "Transform (Modify) Bone - Bone: dyn_front_skirt_02_m"(ComponentPose=@n42, bAlphaBoolEnabled=True, Translation="0,0,10", Rotation="0,0,0", Scale="1,1,1") {
    alphaInputType=Bool
    rotationMode=BMM_Additive
    rotationSpace=BCS_BoneSpace
    translationMode=BMM_Additive
    translationSpace=BCS_BoneSpace
    alphaBoolBlend {blendInTime=0.2, blendOption=Linear, blendOutTime=0.2}
    boneToModify {boneName=dyn_front_skirt_02_m}
    translation {z=10}
   }
   n44 = "Transform (Modify) Bone - Bone: dyn_side_skirt_01_l"!(ComponentPose=@n43, bAlphaBoolEnabled:bool=True, Translation="0,0,2", Rotation="0,0,0.001", Scale="1,1,1") -> (Pose) {
    alphaInputType=Bool
    rotationMode=BMM_Additive
    rotationSpace=BCS_BoneSpace
    translationMode=BMM_Additive
    translationSpace=BCS_BoneSpace
    alphaBoolBlend {blendInTime=0.2, blendOption=Linear, blendOutTime=0.2}
    boneToModify {boneName=dyn_side_skirt_01_l}
    rotation {roll=0.001}
    translation {z=2}
   }
   n45 = "Transform (Modify) Bone - Bone: offset_back_skirt_middle_01_l"(ComponentPose=@n44, bAlphaBoolEnabled=True, Translation="0,0,0", Rotation="10,0,0", Scale="1,1,1") {
    alphaInputType=Bool
    rotationMode=BMM_Additive
    rotationSpace=BCS_BoneSpace
    translationMode=BMM_Additive
    translationSpace=BCS_BoneSpace
    alphaBoolBlend {blendInTime=0.2, blendOption=Linear, blendOutTime=0.2}
    boneToModify {boneName=offset_back_skirt_middle_01_l}
    rotation {pitch=10}
   }
   n46 = "Transform (Modify) Bone - Bone: offset_back_skirt_middle_01_r"(ComponentPose=@n45, bAlphaBoolEnabled=True, Translation="0,0,0", Rotation="10,0,0", Scale="1,1,1") {
    alphaInputType=Bool
    rotationMode=BMM_Additive
    rotationSpace=BCS_BoneSpace
    translationMode=BMM_Additive
    translationSpace=BCS_BoneSpace
    alphaBoolBlend {blendInTime=0.2, blendOption=Linear, blendOutTime=0.2}
    boneToModify {boneName=offset_back_skirt_middle_01_r}
    rotation {pitch=10}
   }
   n47 = "Transform (Modify) Bone - Bone: base_front_skirt_l"(ComponentPose=@n46, bAlphaBoolEnabled=True, Translation="0,0,0", Rotation="5,-5,0", Scale="1,1,1") {
    alphaInputType=Bool
    rotationMode=BMM_Additive
    rotationSpace=BCS_BoneSpace
    alphaBoolBlend {blendInTime=0.2, blendOption=Linear, blendOutTime=0.2}
    boneToModify {boneName=base_front_skirt_l}
    rotation {pitch=5, yaw=-5}
   }
   n48 = "Transform (Modify) Bone - Bone: offset_back_skirt_middle_01_l"(ComponentPose=@n47, bAlphaBoolEnabled=True, Translation="0,0,0", Rotation="-5,0,0", Scale="1,1,1") {
    alphaInputType=Bool
    rotationMode=BMM_Additive
    rotationSpace=BCS_BoneSpace
    alphaBoolBlend {blendInTime=0.2, blendOption=Linear, blendOutTime=0.2}
    boneToModify {boneName=offset_back_skirt_middle_01_l}
    rotation {pitch=-5}
   }
   Save cached pose 'EmotePose'(Pose=Component To Local(ComponentPose=@n48)) {cacheName=EmotePose}
  ]}
 EventGraph {nodes [
   Event BlueprintUpdateAnimation()
   > n1 = Sequence!(execute) -> (then_0, then_1, then_2:Exec, then_3:Exec)
   > n2 = Gravity Override State Function DEPRECATED(CharacterPartAnimInstance=Self-Reference(), GameTimeInSeconds=GetGameTimeInSeconds(), GravityOverrideParams=Get Gravity Override Params())
   > Set GravityOverride(GravityOverride=@n2)
   n3 = Get bIsSkyDiving()
   n4 = Get bIsFalling()
   n5 = Get bIsJumping()
   Set IssJumping(execute=@n1.then_2, IssJumping:bool=AND Boolean(A=NOT Boolean(A=@n3), B=OR Boolean(A=@n5, B=@n4))) -> (Output_Get:bool)
   n6 = Get bIsMoving2D()
   > Set IsMovCrouch(IsMovCrouch:bool=AND Boolean(A=Get bIsCrouching(), B=@n6)) -> (Output_Get:bool)
   n7 = GetAnimInstance(self=GetOwnerSkeletalMeshComponent())
   Is Valid(exec=@n1.then_3, InputObject=@n7)
   > Set isCustomEmotePlaying(isCustomEmotePlaying:bool=Montage_IsPlaying(self=@n7, Montage=/BRCosmetics/Animation/Game/MainPlayer/Emotes/SunBurst/Emote_SunBurst_CMF_Skirt_M)) -> (Output_Get:bool)
   > Set IsBalletSpin(IsBalletSpin:bool=Montage_IsPlaying(self=@n7, Montage=/BRCosmetics/Animation/Game/MainPlayer/Emotes/BalletSpin/Emote_BalletSpin_CMF_M)) -> (Output_Get:bool)
   > Set isMoonGazer(isMoonGazer:bool=Montage_IsPlaying(self=@n7, Montage=/BRCosmetics/Animation/Game/MainPlayer/Emotes/Memory/CMF/Emote_Memory_CMF_M)) -> (Output_Get:bool)
   n8 = Get bIsMoving2D()
   Set CrouchIDLE(execute=@n1.then_1, CrouchIDLE:bool=AND Boolean(A=Get bIsCrouching(), B=NOT Boolean(A=@n8))) -> (Output_Get:bool)
   n9 = Get bIsCrouching()
   > Set IsWalking(IsWalking:bool=AND Boolean(A=NOT Boolean(A=@n9), B=Get bIsMoving2D())) -> (Output_Get:bool)
   NOT Boolean(A=@n6)
   NOT Boolean(A=Get bIsSprinting())
  ]}
]
variables [
 CrouchIDLE {type=boolean}
 Gravity Override Params {type=GravityOverrideParamsStruct}
 GravityOverride {type=Vector}
 IsBalletSpin {type=boolean}
 isCustomEmotePlaying {type=boolean}
 isMoonGazer {type=boolean}
 IsMovCrouch {type=boolean}
 IssJumping {type=boolean}
 IsWalking {type=boolean}
]
