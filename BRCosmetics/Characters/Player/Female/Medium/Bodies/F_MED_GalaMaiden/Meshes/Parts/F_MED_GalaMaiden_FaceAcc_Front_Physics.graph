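# Physics asset for the front face-accessory part of the female GalaMaiden character body, giving its front hair bones collision shapes and joints for secondary motion.
# Three chains (center, left, right) start at small sphere base bodies jointed to a head capsule with free swing and twist.
# Capsule bodies form one segment in the center and two per side, joined by narrow swing limits with twist locked or limited to zero to keep strands from rolling.
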PhysicsAsset /BRCosmetics/Characters/Player/Female/Medium/Bodies/F_MED_GalaMaiden/Meshes/Parts/F_MED_GalaMaiden_FaceAcc_Front_Physics
bodies {
 base_hair_front_c [{shapeType=Sphere, radius=0.505, shapeName=base_hair_front_c_sphere}]
 base_hair_front_l [{shapeType=Sphere, radius=0.505, shapeName=base_hair_front_l_sphere}]
 base_hair_front_r [{shapeType=Sphere, radius=0.505, shapeName=base_hair_front_r_sphere}]
 dyn_hair_front_01_c [{shapeType=Capsule, length=2, radius=2, shapeName=dyn_hair_front_01_c_capsule, center=<3 0 0>, rotation=<rot 90 0 0>}]
 dyn_hair_front_01_l [{shapeType=Capsule, length=2.6852, radius=1.3148, shapeName=dyn_hair_front_01_l_capsule, center=<3 0 0>, rotation=<rot 90 0 0>}]
 dyn_hair_front_01_r [{shapeType=Capsule, length=5, radius=1.5, shapeName=dyn_hair_front_01_r_capsule, center=<4.3447 0 -0.4528>, rotation=<rot 90 0 0>}]
 dyn_hair_front_02_l [{shapeType=Capsule, length=2.872, radius=1.128, shapeName=dyn_hair_front_02_l_capsule, center=<2 0 0>, rotation=<rot 90 0 0>}]
 dyn_hair_front_02_r [{shapeType=Capsule, length=2.9018, radius=1.2734, shapeName=dyn_hair_front_02_r_capsule, center=<3.2806 0 0>, rotation=<rot 90 0 0>}]
 head [{shapeType=Capsule, length=6.5, radius=4.9808, center=<0.7945 4.4779 0.0041>, rotation=<rot 0.2538 -88.3677 -89.5604>}]
}
constraints [
 {
  bone1Name=base_hair_front_c
  bone2Name=head
  swing1LimitDegrees=45
  swing1Motion=Free
  swing2LimitDegrees=45
  swing2Motion=Free
  twistLimitDegrees=45
  twistMotion=Free
 }
 {
  bone1Name=base_hair_front_l
  bone2Name=head
  swing1LimitDegrees=45
  swing1Motion=Free
  swing2LimitDegrees=45
  swing2Motion=Free
  twistLimitDegrees=45
  twistMotion=Free
 }
 {
  bone1Name=base_hair_front_r
  bone2Name=head
  swing1LimitDegrees=45
  swing1Motion=Free
  swing2LimitDegrees=45
  swing2Motion=Free
  twistLimitDegrees=45
  twistMotion=Free
 }
 {
  bone1Name=dyn_hair_front_01_r
  bone2Name=base_hair_front_r
  swing1LimitDegrees=20
  swing2LimitDegrees=20
  twistLimitDegrees=45
  twistMotion=Locked
 }
 {
  bone1Name=dyn_hair_front_02_l
  bone2Name=dyn_hair_front_01_l
  swing1LimitDegrees=15
  swing2LimitDegrees=10
  twistLimitDegrees=45
  twistMotion=Locked
 }
 {
  bone1Name=dyn_hair_front_01_l
  bone2Name=base_hair_front_l
  swing1LimitDegrees=20
  swing2LimitDegrees=10
  twistLimitDegrees=45
  twistMotion=Locked
 }
 {
  bone1Name=dyn_hair_front_01_c
  bone2Name=base_hair_front_c
  swing1LimitDegrees=10
  swing2LimitDegrees=10
  twistLimitDegrees=45
  twistMotion=Locked
 }
 {bone1Name=dyn_hair_front_02_r, bone2Name=dyn_hair_front_01_r, swing1LimitDegrees=10, swing2LimitDegrees=10, twistLimitDegrees=0}
]
