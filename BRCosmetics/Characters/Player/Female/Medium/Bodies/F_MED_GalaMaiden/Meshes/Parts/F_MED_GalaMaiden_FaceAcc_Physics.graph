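# Physics asset for the F_MED_GalaMaiden female character's FaceAcc mesh, defining bodies and joints for its hair.
# Five two-bone dyn_hair capsule chains (center, left, right, both sides) hang from small base_hair spheres on the head.
# Capsules also cover head, neck, upper spine, clavicles and upper arms, linked by free constraints.
# Hair chain joints use tight swing limits and locked twist, except the dyn_hair_01_l root, which is fully free.
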PhysicsAsset /BRCosmetics/Characters/Player/Female/Medium/Bodies/F_MED_GalaMaiden/Meshes/Parts/F_MED_GalaMaiden_FaceAcc_Physics
bodies {
 base_hair_c [{shapeType=Sphere, radius=0.505, shapeName=base_hair_c_sphere}]
 base_hair_l [{shapeType=Sphere, radius=0.505, shapeName=base_hair_l_sphere}]
 base_hair_r [{shapeType=Sphere, radius=0.505, shapeName=base_hair_r_sphere}]
 base_hair_side_l [{shapeType=Sphere, radius=0.505, shapeName=base_hair_side_l_sphere}]
 base_hair_side_r [{shapeType=Sphere, radius=0.505, shapeName=base_hair_side_r_sphere}]
 clavicle_l [{shapeType=Capsule, length=14, radius=6, shapeName=clavicle_l_capsule, center=<6.0073 0 -1.8342>, rotation=<rot -0.0004 -90 100>}]
 clavicle_r [{shapeType=Capsule, length=14, radius=6, shapeName=clavicle_l_capsule, center=<-6.0073 0 1.8342>, rotation=<rot -0.0004 -90 -80>}]
 dyn_hair_01_c [{shapeType=Capsule, length=3, radius=3, shapeName=dyn_hair_01_c_capsule, center=<5 0 0>, rotation=<rot 90 0 0>}]
 dyn_hair_01_l [{shapeType=Capsule, length=3, radius=3, shapeName=dyn_hair_01_l_capsule, center=<5 0 0>, rotation=<rot 90 0 0>}]
 dyn_hair_01_r [{shapeType=Capsule, length=3, radius=3, shapeName=dyn_hair_01_r_capsule, center=<5 0 0>, rotation=<rot 90 0 0>}]
 dyn_hair_02_c [{shapeType=Capsule, length=5.4124, radius=2.0876, shapeName=dyn_hair_02_c_capsule, center=<8.4504 0 0.3391>, rotation=<rot 0 0 90>}]
 dyn_hair_02_l [{shapeType=Capsule, length=3, radius=3, shapeName=dyn_hair_02_l_capsule, center=<5.6515 -0.8951 1.1632>, rotation=<rot 45.0003 -180 -180>}]
 dyn_hair_02_r [{
   shapeType=Capsule
   length=3
   radius=3
   shapeName=dyn_hair_02_r_capsule
   center=<6.5517 0.0069 1.6783>
   rotation=<rot 60.4676 -163.8798 -162.6319>
  }]
 dyn_hair_side_01_l [{
   shapeType=Capsule
   length=7.2479
   radius=2.6162
   shapeName=dyn_hairBack_01_l_capsule
   center=<4.4312 0.5479 1.1787>
   rotation=<rot -90 179.9989 -179.9987>
  }]
 dyn_hair_side_01_r [{shapeType=Capsule, length=2, radius=2, shapeName=dyn_hair_side_01_r_capsule, center=<3.9244 0 0>, rotation=<rot 90 0 0>}]
 dyn_hair_side_02_l [{
   shapeType=Capsule
   length=3.814
   radius=2.7003
   shapeName=dyn_hairBack_02_l_capsule
   center=<3.5507 -1.0983 2.2056>
   rotation=<rot 0.0001 91.2692 179.9998>
  }]
 dyn_hair_side_02_r [{
   shapeType=Capsule
   length=1.4863
   radius=1.71
   shapeName=dyn_hair_side_02_r_capsule
   center=<5.2321 0.5318 2.5867>
   rotation=<rot 90 180 180>
  }]
 head [{shapeType=Capsule, length=6.5, radius=4.9808, center=<0.7945 4.4779 0.0041>, rotation=<rot 0.2538 -88.3677 -89.5604>}]
 neck_01 [{shapeType=Capsule, length=6.005, radius=4.505, shapeName=neck_01_capsule, center=<3.2328 0 0>, rotation=<rot 0 -90.0001 90>}]
 spine_05 [{shapeType=Capsule, length=15.005, radius=10.505, shapeName=spine_05_capsule, center=<-1.2162 0 0>, rotation=<rot 0 -90 0>}]
 upperarm_l [{shapeType=Capsule, length=33.5, radius=4.5, shapeName=upperarm_l_capsule, center=<15.861 0 -0.5299>, rotation=<rot 85 -179.9986 -179.9987>}]
 upperarm_r [{
   shapeType=Capsule
   length=34.2098
   radius=3.7902
   shapeName=upperarm_l_capsule
   center=<-15.861 0.5113 0.5299>
   rotation=<rot 85 -179.9986 0.0013>
  }]
}
constraints [
 {
  bone1Name=base_hair_c
  bone2Name=head
  swing1LimitDegrees=45
  swing1Motion=Free
  swing2LimitDegrees=45
  swing2Motion=Free
  twistLimitDegrees=45
  twistMotion=Free
 }
 {
  bone1Name=base_hair_l
  bone2Name=head
  swing1LimitDegrees=45
  swing1Motion=Free
  swing2LimitDegrees=45
  swing2Motion=Free
  twistLimitDegrees=45
  twistMotion=Free
 }
 {
  bone1Name=base_hair_r
  bone2Name=head
  swing1LimitDegrees=45
  swing1Motion=Free
  swing2LimitDegrees=45
  swing2Motion=Free
  twistLimitDegrees=45
  twistMotion=Free
 }
 {
  bone1Name=base_hair_side_l
  bone2Name=head
  swing1LimitDegrees=45
  swing1Motion=Free
  swing2LimitDegrees=45
  swing2Motion=Free
  twistLimitDegrees=45
  twistMotion=Free
 }
 {
  bone1Name=base_hair_side_r
  bone2Name=head
  swing1LimitDegrees=45
  swing1Motion=Free
  swing2LimitDegrees=45
  swing2Motion=Free
  twistLimitDegrees=45
  twistMotion=Free
 }
 {
  bone1Name=head
  bone2Name=neck_01
  swing1LimitDegrees=45
  swing1Motion=Free
  swing2LimitDegrees=45
  swing2Motion=Free
  twistLimitDegrees=45
  twistMotion=Free
 }
 {
  bone1Name=clavicle_l
  bone2Name=spine_05
  swing1LimitDegrees=45
  swing1Motion=Free
  swing2LimitDegrees=45
  swing2Motion=Free
  twistLimitDegrees=45
  twistMotion=Free
 }
 {
  bone1Name=neck_01
  bone2Name=spine_05
  swing1LimitDegrees=45
  swing1Motion=Free
  swing2LimitDegrees=45
  swing2Motion=Free
  twistLimitDegrees=45
  twistMotion=Free
 }
 {
  bone1Name=dyn_hair_side_01_r
  bone2Name=base_hair_side_r
  swing1LimitDegrees=5
  swing2LimitDegrees=0
  twistLimitDegrees=45
  twistMotion=Locked
 }
 {
  bone1Name=dyn_hair_side_02_r
  bone2Name=dyn_hair_side_01_r
  swing1LimitDegrees=15
  swing2LimitDegrees=5
  twistLimitDegrees=45
  twistMotion=Locked
 }
 {
  bone1Name=dyn_hair_side_02_l
  bone2Name=dyn_hair_side_01_l
  swing1LimitDegrees=10
  swing2LimitDegrees=5
  twistLimitDegrees=45
  twistMotion=Locked
 }
 {
  bone1Name=dyn_hair_side_01_l
  bone2Name=base_hair_side_l
  swing1LimitDegrees=10
  swing2LimitDegrees=0
  twistLimitDegrees=45
  twistMotion=Locked
 }
 {bone1Name=dyn_hair_02_r, bone2Name=dyn_hair_01_r, swing1LimitDegrees=10, swing2LimitDegrees=20, twistLimitDegrees=45, twistMotion=Locked}
 {bone1Name=dyn_hair_01_r, bone2Name=base_hair_r, swing1LimitDegrees=0, swing2LimitDegrees=20, twistLimitDegrees=45, twistMotion=Locked}
 {bone1Name=dyn_hair_02_c, bone2Name=dyn_hair_01_c, swing1LimitDegrees=5, swing2LimitDegrees=20, twistLimitDegrees=45, twistMotion=Locked}
 {bone1Name=dyn_hair_01_c, bone2Name=base_hair_c, swing1LimitDegrees=0, swing2LimitDegrees=10, twistLimitDegrees=45, twistMotion=Locked}
 {
  bone1Name=clavicle_r
  bone2Name=spine_05
  swing1LimitDegrees=45
  swing1Motion=Free
  swing2LimitDegrees=45
  swing2Motion=Free
  twistLimitDegrees=45
  twistMotion=Free
 }
 {
  bone1Name=upperarm_l
  bone2Name=clavicle_l
  swing1LimitDegrees=45
  swing1Motion=Free
  swing2LimitDegrees=45
  swing2Motion=Free
  twistLimitDegrees=45
  twistMotion=Free
 }
 {
  bone1Name=upperarm_r
  bone2Name=clavicle_r
  swing1LimitDegrees=45
  swing1Motion=Free
  swing2LimitDegrees=45
  swing2Motion=Free
  twistLimitDegrees=45
  twistMotion=Free
 }
 {bone1Name=dyn_hair_02_l, bone2Name=dyn_hair_01_l, swing1LimitDegrees=10, swing2LimitDegrees=20, twistLimitDegrees=45, twistMotion=Locked}
 {
  bone1Name=dyn_hair_01_l
  bone2Name=base_hair_l
  swing1LimitDegrees=45
  swing1Motion=Free
  swing2LimitDegrees=45
  swing2Motion=Free
  twistLimitDegrees=45
  twistMotion=Free
 }
]
